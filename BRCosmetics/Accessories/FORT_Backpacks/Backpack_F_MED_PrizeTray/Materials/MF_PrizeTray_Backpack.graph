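# Material function for the PrizeTray backpack that blends an animated chalk marking over Lit and Shade color inputs.
# Blueprint-fed game time eases progress toward CurrentStage, with an FEDemo floor and a ReactivityOn multiplier.
# Its fractional part splits into overlapping windows, whose curve ramps drive MF_MaskShifting to reveal and shift the packed ChalkMask channels.
# The mask is rotated, positioned and box-clipped in UV space before tinting both outputs with ChalkColor.
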MaterialFunction /BRCosmetics/Accessories/FORT_Backpacks/Backpack_F_MED_PrizeTray/Materials/MF_PrizeTray_Backpack
graph {expressions {
  MaterialExpressionAdd_0 {
   class=MaterialExpressionAdd
   inputs [
    A {from=MaterialExpressionStep_0}
    B {from=""}
   ]
   properties {desc=Debug}
  }
  MaterialExpressionAdd_1 {
   class=MaterialExpressionAdd
   inputs [
    A {from=MaterialExpressionMultiply_1}
    B {from=MaterialExpressionComponentMask_1}
   ]
  }
  MaterialExpressionAdd_4 {
   class=MaterialExpressionAdd
   inputs [
    A {from=MaterialExpressionFrac_2}
    B {from=""}
   ]
   properties {constB=-0.025}
  }
  MaterialExpressionAdd_5 {
   class=MaterialExpressionAdd
   inputs [
    A {from=MaterialExpressionScalarParameter_16}
    B {from=""}
   ]
   properties {constB=0}
  }
  MaterialExpressionClamp_1 {
   class=MaterialExpressionClamp
   inputs [
    None {from=MaterialExpressionAdd_1, output=""}
    Min {from="", output=""}
    Max {from="", output=""}
   ]
   properties {}
  }
  MaterialExpressionColorRamp_16 {class=MaterialExpressionColorRamp, inputs [None {from=MaterialExpressionSaturate_0, output=""}], properties {colorCurve="<self>.ColorCurve"}}
  MaterialExpressionColorRamp_18 {class=MaterialExpressionColorRamp, inputs [None {from=MaterialExpressionMultiply_8, output=""}], properties {colorCurve="<self>.ColorCurve"}}
  MaterialExpressionComponentMask_0 {class=MaterialExpressionComponentMask, inputs [None {from=MaterialExpressionColorRamp_16}], properties {r=true}}
  MaterialExpressionComponentMask_1 {class=MaterialExpressionComponentMask, inputs [None {from=MaterialExpressionVectorParameter_0, output=RGB}], properties {g=true, r=true}}
  MaterialExpressionComponentMask_2 {class=MaterialExpressionComponentMask, inputs [None {from=MaterialExpressionLinearInterpolate_11}], properties {r=true}}
  MaterialExpressionComponentMask_3 {class=MaterialExpressionComponentMask, inputs [None {from=MaterialExpressionColorRamp_18}], properties {r=true}}
  MaterialExpressionConstant2Vector_0 {class=MaterialExpressionConstant2Vector, properties {g=0.5}}
  MaterialExpressionConstant2Vector_3 {class=MaterialExpressionConstant2Vector, properties {g=1.2, r=0.8}}
  MaterialExpressionConstant2Vector_6 {class=MaterialExpressionConstant2Vector, properties {g=0.5, r=0.5}}
  MaterialExpressionConstant4Vector_1 {class=MaterialExpressionConstant4Vector, inputs [], properties {constant=<rgba 0 0.1 0 1>}}
  MaterialExpressionConstant4Vector_2 {class=MaterialExpressionConstant4Vector, inputs [], properties {constant=<rgba 0 1 0 1>}}
  MaterialExpressionConstant_0 {class=MaterialExpressionConstant, properties {r=0.25}}
  MaterialExpressionConstant_1 {class=MaterialExpressionConstant, properties {desc=Start Time}}
  MaterialExpressionConstant_2 {class=MaterialExpressionConstant, properties {desc=Start Time, r=0.15}}
  MaterialExpressionConstant_3 {class=MaterialExpressionConstant, properties {r=0.1}}
  MaterialExpressionConstant_4 {class=MaterialExpressionConstant, properties {desc=Start Time, r=0.7}}
  MaterialExpressionConvert_0 {
   class=MaterialExpressionConvert
   inputs [
    None {from=MaterialExpressionComponentMask_2}
    None {from=""}
   ]
   properties {
    nodeName=Make Vector2
    convertInputs [
     {
      type=Scalar
      expressionInput {
       expression=&self:MaterialExpressionComponentMask_2
       inputName=None
       mask=0
       maskA=0
       maskB=0
       maskG=0
       maskR=0
       outputIndex=0
      }
     }
     {type=Scalar, expressionInput {expression=null, inputName=None, mask=0, maskA=0, maskB=0, maskG=0, maskR=0, outputIndex=0}}
    ]
    convertOutputs [{type=Vector2}]
   }
  }
  MaterialExpressionDivide_2 {
   class=MaterialExpressionDivide
   inputs [
    A {from=MaterialExpressionSubtract_3}
    B {from=""}
   ]
   properties {constB=0.15, desc=Duration}
  }
  MaterialExpressionDivide_3 {
   class=MaterialExpressionDivide
   inputs [
    A {from=MaterialExpressionSubtract_4}
    B {from=""}
   ]
   properties {constB=0.55, desc=Duration}
  }
  MaterialExpressionDivide_4 {
   class=MaterialExpressionDivide
   inputs [
    A {from=MaterialExpressionSubtract_5}
    B {from=""}
   ]
   properties {constB=0.3, desc=Duration}
  }
  MaterialExpressionDivide_5 {
   class=MaterialExpressionDivide
   inputs [
    A {from=MaterialExpressionSubtract_7}
    B {from=MaterialExpressionAdd_5}
   ]
  }
  MaterialExpressionFrac_1 {class=MaterialExpressionFrac, inputs [None {from=MaterialExpressionMultiply_56, output=""}], properties {}}
  MaterialExpressionFrac_2 {class=MaterialExpressionFrac, inputs [None {from=MaterialExpressionTextureSampleParameter2D_0, output=A}], properties {}}
  MaterialExpressionFunctionInput_0 {class=MaterialExpressionFunctionInput, inputs [Preview {from="", output=""}], properties {bUsePreviewValueAsDefault=true, inputName=In Lit}}
  MaterialExpressionFunctionInput_1 {
   class=MaterialExpressionFunctionInput
   inputs [Preview {from="", output=""}]
   properties {bUsePreviewValueAsDefault=true, inputName=In Shade, sortPriority=1}
  }
  MaterialExpressionFunctionOutput_0 {
   class=MaterialExpressionFunctionOutput
   inputs [None {from=MaterialExpressionLinearInterpolate_0, output=""}]
   properties {outputName=Lit Out}
  }
  MaterialExpressionFunctionOutput_1 {
   class=MaterialExpressionFunctionOutput
   inputs [None {from=MaterialExpressionLinearInterpolate_2, output=""}]
   properties {outputName=Shade Out, sortPriority=1}
  }
  MaterialExpressionLength_0 {class=MaterialExpressionLength, inputs [None {from=MaterialExpressionSubtract_2, output=""}], properties {}}
  MaterialExpressionLinearInterpolate_0 {
   class=MaterialExpressionLinearInterpolate
   inputs [
    A {from=MaterialExpressionFunctionInput_0}
    B {from=MaterialExpressionLinearInterpolate_4}
    Alpha {from=MaterialExpressionMultiply_11}
   ]
  }
  MaterialExpressionLinearInterpolate_10 {
   class=MaterialExpressionLinearInterpolate
   inputs [
    A {from=MaterialExpressionLinearInterpolate_12}
    B {from=MaterialExpressionReroute_1}
    Alpha {from=MaterialExpressionSaturate_3}
   ]
   properties {desc=Wait}
  }
  MaterialExpressionLinearInterpolate_11 {
   class=MaterialExpressionLinearInterpolate
   inputs [
    A {from=MaterialExpressionLinearInterpolate_10}
    B {from=MaterialExpressionReroute_2}
    Alpha {from=MaterialExpressionSaturate_5}
   ]
   properties {desc=Wait}
  }
  MaterialExpressionLinearInterpolate_12 {
   class=MaterialExpressionLinearInterpolate
   inputs [
    A {from=""}
    B {from=MaterialExpressionReroute_0}
    Alpha {from=MaterialExpressionSaturate_0}
   ]
   properties {desc=Wait}
  }
  MaterialExpressionLinearInterpolate_2 {
   class=MaterialExpressionLinearInterpolate
   inputs [
    A {from=MaterialExpressionFunctionInput_1}
    B {from=MaterialExpressionLinearInterpolate_6}
    Alpha {from=MaterialExpressionMultiply_11}
   ]
  }
  MaterialExpressionLinearInterpolate_4 {
   class=MaterialExpressionLinearInterpolate
   inputs [
    A {from=MaterialExpressionMultiply_4}
    B {from=MaterialExpressionVectorParameter_1, output=RGB}
    Alpha {from=""}
   ]
   properties {constAlpha=0.85}
  }
  MaterialExpressionLinearInterpolate_5 {
   class=MaterialExpressionLinearInterpolate
   inputs [
    A {from=MaterialExpressionTextureSampleParameter2D_0, output=G}
    B {from=MaterialExpressionMultiply_2}
    Alpha {from=MaterialExpressionSaturate_2}
   ]
  }
  MaterialExpressionLinearInterpolate_6 {
   class=MaterialExpressionLinearInterpolate
   inputs [
    A {from=MaterialExpressionMultiply_6}
    B {from=MaterialExpressionVectorParameter_1, output=RGB}
    Alpha {from=""}
   ]
   properties {constAlpha=0.85}
  }
  MaterialExpressionLinearInterpolate_8 {
   class=MaterialExpressionLinearInterpolate
   inputs [
    A {from=""}
    B {from=MaterialExpressionComponentMask_3}
    Alpha {from=MaterialExpressionSaturate_3}
   ]
   properties {desc=Wait}
  }
  MaterialExpressionLinearInterpolate_9 {
   class=MaterialExpressionLinearInterpolate
   inputs [
    A {from=MaterialExpressionSubtract_11}
    B {from=MaterialExpressionScalarParameter_6}
    Alpha {from=MaterialExpressionNamedRerouteUsage_13}
   ]
   properties {desc=Native lerp toward}
  }
  MaterialExpressionMaterialFunctionCall_1 {
   class=MaterialExpressionMaterialFunctionCall
   inputs [
    UVs {from=MaterialExpressionTextureCoordinate_1}
    Rotation Center {from=""}
    "Rotation Angle (0-1)" {from=MaterialExpressionScalarParameter_0}
   ]
   properties {materialFunction=&CustomRotator}
  }
  MaterialExpressionMaterialFunctionCall_4 {
   class=MaterialExpressionMaterialFunctionCall
   inputs [
    Mask In {from=MaterialExpressionComponentMask_0}
    Amount {from=MaterialExpressionMultiply_13}
    Range {from=MaterialExpressionConstant4Vector_1, output=G}
    Start Offset {from=MaterialExpressionConstant4Vector_1, output=G}
    Sharpness {from=MaterialExpressionConstant4Vector_1, output=G}
   ]
   properties {materialFunction=/Game/Global/MaterialFunctions/MF_MaskShifting}
  }
  MaterialExpressionMaterialFunctionCall_5 {
   class=MaterialExpressionMaterialFunctionCall
   inputs [
    Mask In {from=MaterialExpressionLinearInterpolate_8}
    Amount {from=MaterialExpressionMultiply_3}
    Range {from=MaterialExpressionConstant4Vector_2, output=G}
    Start Offset {from=MaterialExpressionConstant4Vector_2, output=G}
    Sharpness {from=MaterialExpressionConstant4Vector_2, output=G}
   ]
   properties {materialFunction=&MF_MaskShifting}
  }
  MaterialExpressionMaterialFunctionCall_6 {
   class=MaterialExpressionMaterialFunctionCall
   inputs [
    A {from=MaterialExpressionAdd_1}
    B {from=MaterialExpressionConstant2Vector_6}
    Bounds {from=MaterialExpressionConstant2Vector_3}
    Edge Falloff {from=MaterialExpressionConstant_3}
   ]
   properties {materialFunction=/Engine/Functions/Engine_MaterialFunctions01/Gradient/BoxMask-2D}
  }
  MaterialExpressionMax_0 {
   class=MaterialExpressionMax
   inputs [
    A {from=MaterialExpressionLinearInterpolate_9}
    B {from=MaterialExpressionScalarParameter_1}
   ]
   properties {constB=0}
  }
  MaterialExpressionMultiply_0 {
   class=MaterialExpressionMultiply
   inputs [
    A {from=MaterialExpressionTextureSampleParameter2D_0, output=R}
    B {from=MaterialExpressionFunctionInput_0}
   ]
  }
  MaterialExpressionMultiply_1 {
   class=MaterialExpressionMultiply
   inputs [
    A {from=MaterialExpressionMaterialFunctionCall_1, output=Rotated Values}
    B {from=""}
   ]
   properties {constB=2.5}
  }
  MaterialExpressionMultiply_11 {
   class=MaterialExpressionMultiply
   inputs [
    A {from=MaterialExpressionMaterialFunctionCall_6, output=Result}
    B {from=MaterialExpressionMultiply_7}
   ]
  }
  MaterialExpressionMultiply_13 {
   class=MaterialExpressionMultiply
   inputs [
    A {from=MaterialExpressionStep_2}
    B {from=MaterialExpressionAdd_4}
   ]
  }
  MaterialExpressionMultiply_18 {
   class=MaterialExpressionMultiply
   inputs [
    A {from=MaterialExpressionStep_3}
    B {from=MaterialExpressionPower_0}
   ]
  }
  MaterialExpressionMultiply_2 {
   class=MaterialExpressionMultiply
   inputs [
    A {from=MaterialExpressionMultiply_18}
    B {from=MaterialExpressionSaturate_1}
   ]
  }
  MaterialExpressionMultiply_3 {
   class=MaterialExpressionMultiply
   inputs [
    A {from=MaterialExpressionTextureSampleParameter2D_0, output=B}
    B {from=""}
   ]
   properties {constB=0.55}
  }
  MaterialExpressionMultiply_4 {
   class=MaterialExpressionMultiply
   inputs [
    A {from=MaterialExpressionMultiply_0}
    B {from=MaterialExpressionVectorParameter_1, output=RGB}
   ]
  }
  MaterialExpressionMultiply_5 {
   class=MaterialExpressionMultiply
   inputs [
    A {from=MaterialExpressionTextureSampleParameter2D_0, output=R}
    B {from=MaterialExpressionFunctionInput_1}
   ]
  }
  MaterialExpressionMultiply_56 {
   class=MaterialExpressionMultiply
   inputs [
    A {from=MaterialExpressionMax_0}
    B {from=MaterialExpressionScalarParameter_8}
   ]
  }
  MaterialExpressionMultiply_6 {
   class=MaterialExpressionMultiply
   inputs [
    A {from=MaterialExpressionMultiply_5}
    B {from=MaterialExpressionVectorParameter_1, output=RGB}
   ]
  }
  MaterialExpressionMultiply_7 {
   class=MaterialExpressionMultiply
   inputs [
    A {from=MaterialExpressionLinearInterpolate_5}
    B {from=MaterialExpressionSaturate_2}
   ]
  }
  MaterialExpressionMultiply_8 {
   class=MaterialExpressionMultiply
   inputs [
    A {from=MaterialExpressionSaturate_5}
    B {from=MaterialExpressionSaturate_5}
   ]
  }
  MaterialExpressionMultiply_9 {
   class=MaterialExpressionMultiply
   inputs [
    A {from=MaterialExpressionConvert_0}
    B {from=""}
   ]
   properties {constB=0.4}
  }
  MaterialExpressionNamedRerouteDeclaration_26 {
   class=MaterialExpressionNamedRerouteDeclaration
   inputs [None {from=MaterialExpressionSaturate_4}]
   properties {name=TimeNative, nodeColor=<rgba 0.2891 1 0 1>}
  }
  MaterialExpressionNamedRerouteUsage_13 {class=MaterialExpressionNamedRerouteUsage}
  MaterialExpressionOneMinus_3 {class=MaterialExpressionOneMinus, inputs [None {from=MaterialExpressionMaterialFunctionCall_5, output=Result}]}
  MaterialExpressionPower_0 {
   class=MaterialExpressionPower
   inputs [
    Base {from=MaterialExpressionTextureSampleParameter2D_0, output=R}
    Exp {from=""}
   ]
  }
  MaterialExpressionReroute_0 {class=MaterialExpressionReroute, inputs [None {from=""}], properties {desc=TestRamp}}
  MaterialExpressionReroute_1 {class=MaterialExpressionReroute, inputs [None {from=""}], properties {desc=TestRamp}}
  MaterialExpressionReroute_2 {class=MaterialExpressionReroute, inputs [None {from=""}], properties {desc=TestRamp}}
  MaterialExpressionSaturate_0 {class=MaterialExpressionSaturate, inputs [None {from=MaterialExpressionDivide_2}]}
  MaterialExpressionSaturate_1 {class=MaterialExpressionSaturate, inputs [None {from=MaterialExpressionMaterialFunctionCall_4, output=Result}]}
  MaterialExpressionSaturate_2 {class=MaterialExpressionSaturate, inputs [None {from=MaterialExpressionOneMinus_3}]}
  MaterialExpressionSaturate_3 {class=MaterialExpressionSaturate, inputs [None {from=MaterialExpressionDivide_3}]}
  MaterialExpressionSaturate_4 {class=MaterialExpressionSaturate, inputs [None {from=MaterialExpressionDivide_5}]}
  MaterialExpressionSaturate_5 {class=MaterialExpressionSaturate, inputs [None {from=MaterialExpressionDivide_4}]}
  MaterialExpressionScalarParameter_0 {class=MaterialExpressionScalarParameter, properties {defaultValue=0.5, group=PrizeTray, parameterName=Rotation}}
  MaterialExpressionScalarParameter_1 {class=MaterialExpressionScalarParameter, properties {group=PrizeTray, parameterName=FEDemo, sortPriority=3}}
  MaterialExpressionScalarParameter_15 {class=MaterialExpressionScalarParameter, properties {defaultValue=-99999, group=PrizeTray, parameterName=GameTimeFromBP, sortPriority=100}}
  MaterialExpressionScalarParameter_16 {class=MaterialExpressionScalarParameter, properties {defaultValue=1, group=PrizeTray, parameterName=Animation_Duration, sortPriority=101}}
  MaterialExpressionScalarParameter_6 {class=MaterialExpressionScalarParameter, properties {group=PrizeTray, parameterName=CurrentStage, sortPriority=3}}
  MaterialExpressionScalarParameter_8 {class=MaterialExpressionScalarParameter, properties {defaultValue=1, parameterName=ReactivityOn}}
  MaterialExpressionStep_0 {
   class=MaterialExpressionStep
   inputs [
    Y {from=MaterialExpressionLength_0, output=""}
    X {from="", output=""}
   ]
   properties {constX=0.01}
  }
  MaterialExpressionStep_2 {
   class=MaterialExpressionStep
   inputs [
    Y {from="", output=""}
    X {from=MaterialExpressionTextureSampleParameter2D_0, output=A}
   ]
   properties {constY=0.001}
  }
  MaterialExpressionStep_3 {
   class=MaterialExpressionStep
   inputs [
    Y {from="", output=""}
    X {from=MaterialExpressionPower_0, output=""}
   ]
   properties {constY=0.001}
  }
  MaterialExpressionSubtract_0 {
   class=MaterialExpressionSubtract
   inputs [
    A {from=MaterialExpressionTextureCoordinate_0}
    B {from=MaterialExpressionConstant2Vector_0}
   ]
   properties {constB=0.5}
  }
  MaterialExpressionSubtract_11 {
   class=MaterialExpressionSubtract
   inputs [
    A {from=MaterialExpressionScalarParameter_6}
    B {from=""}
   ]
  }
  MaterialExpressionSubtract_2 {
   class=MaterialExpressionSubtract
   inputs [
    A {from=MaterialExpressionSubtract_0}
    B {from=MaterialExpressionMultiply_9}
   ]
  }
  MaterialExpressionSubtract_3 {
   class=MaterialExpressionSubtract
   inputs [
    A {from=MaterialExpressionFrac_1}
    B {from=MaterialExpressionConstant_1}
   ]
  }
  MaterialExpressionSubtract_4 {
   class=MaterialExpressionSubtract
   inputs [
    A {from=MaterialExpressionFrac_1}
    B {from=MaterialExpressionConstant_2}
   ]
  }
  MaterialExpressionSubtract_5 {
   class=MaterialExpressionSubtract
   inputs [
    A {from=MaterialExpressionFrac_1}
    B {from=MaterialExpressionConstant_4}
   ]
  }
  MaterialExpressionSubtract_7 {
   class=MaterialExpressionSubtract
   inputs [
    A {from=MaterialExpressionTime_3}
    B {from=MaterialExpressionScalarParameter_15}
   ]
  }
  MaterialExpressionTextureCoordinate_0 {class=MaterialExpressionTextureCoordinate}
  MaterialExpressionTextureCoordinate_1 {class=MaterialExpressionTextureCoordinate}
  MaterialExpressionTextureSampleParameter2D_0 {
   class=MaterialExpressionTextureSampleParameter2D
   inputs [
    UVs {from=MaterialExpressionClamp_1, output=""}
    Apply View MipBias {from="", output=""}
   ]
   properties {
    parameterName=ChalkMask
    samplerSource=SSM_Clamp_WorldGroupSettings
    texture=/BRCosmetics/Accessories/FORT_Backpacks/Backpack_F_MED_PrizeTray/Textures/T_PrizeTray_Backpack_Mask
   }
  }
  MaterialExpressionTime_3 {class=MaterialExpressionTime}
  MaterialExpressionVectorParameter_0 {class=MaterialExpressionVectorParameter, properties {group=PrizeTray, parameterName=Postion, defaultValue=<rgba -1.48 -0.325 0 1>}}
  MaterialExpressionVectorParameter_1 {class=MaterialExpressionVectorParameter, properties {parameterName=ChalkColor, defaultValue=<rgba 0.8 0.8 0.75 1>}}
 }}
