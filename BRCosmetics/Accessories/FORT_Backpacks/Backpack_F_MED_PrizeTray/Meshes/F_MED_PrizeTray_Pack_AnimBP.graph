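# Animation Blueprint for the PrizeTray backpack cosmetic, posing it and adding secondary physics motion.
# The update event derives movement-state values, spine and head twist alphas, and a parameter-asset gravity override from the character.
# The AnimGraph chains additive root, base_stand and pack offsets into AnimDynamics for chalkboard, brush, chalk and stand.
# Non-skydive falling enables brush and chalk physics and offsets; airborne skydive floating tilts base_board and stand.
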AnimBlueprint /BRCosmetics/Accessories/FORT_Backpacks/Backpack_F_MED_PrizeTray/Meshes/F_MED_PrizeTray_Pack_AnimBP
target_skeleton=&/F_MED_PrizeTray_Pack_Skeleton
graphs [
 AnimGraph {nodes [
   n1 = "Transform (Modify) Bone - Bone: root"(ComponentPose=Local To Component(LocalPose=Input Pose()), Translation="0,-3,0", Rotation="0,0,15") {
    rotationMode=BMM_Additive
    rotationSpace=BCS_BoneSpace
    translationMode=BMM_Additive
    translationSpace=BCS_BoneSpace
    alphaScaleBiasClamp {bInterpResult=true}
    boneToModify {boneName=root}
    rotation {roll=15}
    translation {y=-3}
   }
   n2 = "Transform (Modify) Bone - Bone: base_stand"(ComponentPose=@n1, Alpha=1, Translation="0,0,0", Rotation="8,0,0", Scale="1,1,1") {
    rotationMode=BMM_Additive
    rotationSpace=BCS_BoneSpace
    translationMode=BMM_Additive
    translationSpace=BCS_BoneSpace
    alphaScaleBiasClamp {bInterpResult=true, interpSpeedDecreasing=20}
    boneToModify {boneName=base_stand}
    rotation {pitch=8}
   }
   n3 = "Transform (Modify) Bone - Bone: root"(ComponentPose=@n2, Translation="0,1,0", Rotation="0,0,20") {
    rotationMode=BMM_Additive
    rotationSpace=BCS_BoneSpace
    translationMode=BMM_Additive
    translationSpace=BCS_BoneSpace
    boneToModify {boneName=root}
    rotation {roll=20}
    translation {y=1}
   }
   n4 = "Transform (Modify) Bone - Bone: base_stand"(ComponentPose=@n3, Alpha=1, Translation="0,0,0", Rotation="8,0,0", Scale="1,1,1") {
    rotationMode=BMM_Additive
    rotationSpace=BCS_BoneSpace
    translationMode=BMM_Additive
    translationSpace=BCS_BoneSpace
    alphaScaleBiasClamp {bInterpResult=true, interpSpeedDecreasing=20}
    boneToModify {boneName=base_stand}
    rotation {pitch=8}
   }
   n5 = "Transform (Modify) Bone - Bone: pack"(ComponentPose=@n4, Translation="0,-2,0", Rotation="0,0,5") {
    rotationMode=BMM_Additive
    rotationSpace=BCS_BoneSpace
    translationMode=BMM_Additive
    translationSpace=BCS_BoneSpace
    alphaScaleBiasClamp {bInterpResult=true, interpSpeedDecreasing=8}
    boneToModify {boneName=pack}
    rotation {roll=5}
    translation {y=-2}
   }
   n6 = "Transform (Modify) Bone - Bone: pack"(ComponentPose=@n5, Translation="0,-2,0", Rotation="0,0,10") {
    rotationMode=BMM_Additive
    rotationSpace=BCS_BoneSpace
    translationMode=BMM_Additive
    translationSpace=BCS_BoneSpace
    boneToModify {boneName=pack}
    rotation {roll=10}
    translation {y=-2}
   }
   n7 = "Transform (Modify) Bone - Bone: pack"(ComponentPose=@n6, Translation="0,-3,0", Rotation="0,0,10") {
    rotationMode=BMM_Additive
    rotationSpace=BCS_BoneSpace
    translationMode=BMM_Additive
    translationSpace=BCS_BoneSpace
    alphaScaleBiasClamp {bInterpResult=true, interpSpeedDecreasing=8}
    boneToModify {boneName=pack}
    rotation {roll=10}
    translation {y=-3}
   }
   n8 = "Transform (Modify) Bone - Bone: pack"(ComponentPose=@n7, Translation="0,-2,0", Rotation="0,30,0") {
    rotationMode=BMM_Additive
    rotationSpace=BCS_BoneSpace
    translationMode=BMM_Additive
    translationSpace=BCS_BoneSpace
    alphaScaleBiasClamp {bInterpResult=true, interpSpeedDecreasing=20, interpSpeedIncreasing=20}
    boneToModify {boneName=pack}
    rotation {yaw=30}
    translation {y=-2}
   }
   n9 = "Transform (Modify) Bone - Bone: pack"(ComponentPose=@n8, Translation="0,-2,0", Rotation="0,-20,0") {
    rotationMode=BMM_Additive
    rotationSpace=BCS_BoneSpace
    translationMode=BMM_Additive
    translationSpace=BCS_BoneSpace
    alphaScaleBiasClamp {bInterpResult=true}
    boneToModify {boneName=pack}
    rotation {yaw=-20}
    translation {y=-2}
   }
   n10 = "Transform (Modify) Bone - Bone: root"(ComponentPose=@n9, Translation="0,0,-9", Rotation="0,0,0") {
    rotationMode=BMM_Additive
    rotationSpace=BCS_BoneSpace
    translationMode=BMM_Additive
    translationSpace=BCS_BoneSpace
    alphaScaleBiasClamp {bInterpResult=true, interpSpeedIncreasing=5}
    boneToModify {boneName=root}
    translation {z=-9}
   }
   n11 = Anim Dynamics - Bone: chalkboard(ComponentPose=@n10, Alpha=1, ExternalForce="0,0,0") {
    angularDampingOverride=1
    angularSpringConstant=20000
    bAngularSpring=true
    bOverrideAngularDamping=true
    bUseGravityOverride=true
    gravityScale=0.35
    simulationSpace=BoneRelative
    boundBone {boneName=chalkboard}
    componentAppliedLinearAccClamp=<3000 3000 3000>
    componentLinearAccScale=<0.5 0.5 0.5>
    componentLinearVelScale=<0.5 0.5 0.5>
    gravityOverride {z=-980}
    physicsBodyDefinitions [{
      boundBone {boneName=chalkboard}
      boxExtents=<40 40 40>
      constraintSetup {angularLimitsMax=<0 20 0>, angularLimitsMin=<0 -1 0>, angularTarget=<1 0 0>}
      localJointOffset=<-3 0 0>
     }]
    relativeSpaceBone {boneName=base_board}
   }
   n12 = Get Falling ground() -> (Falling ground:bool)
   n13 = Anim Dynamics - Bone: brush!(ComponentPose=@n11, bAlphaBoolEnabled:bool=@n12, ExternalForce="0,0,0") -> (Pose) {
    alphaInputType=Bool
    angularDampingOverride=0.888
    angularSpringConstant=5000
    bAngularSpring=true
    bLinearSpring=true
    bOverrideAngularDamping=true
    gravityScale=1.1
    simulationSpace=BoneRelative
    alphaBoolBlend {blendInTime=0.25, blendOutTime=0.2}
    boundBone {boneName=brush}
    componentAppliedLinearAccClamp=<3000 3000 3000>
    componentLinearAccScale=<0.4 0.4 0.4>
    componentLinearVelScale=<0.4 0.4 0.4>
    gravityOverride {z=-980}
    physicsBodyDefinitions [{
      boundBone {boneName=brush}
      boxExtents=<40 40 40>
      constraintSetup {angularLimitsMax=<0 0 15>, angularLimitsMin=<0 0 -20>, angularTarget=<1 0 0>}
      localJointOffset=<2 0 0>
     }]
    relativeSpaceBone {boneName=pack}
    simSpaceSettings {maxAngularAcceleration=100, maxAngularVelocity=10, simSpaceAngularAlpha=1}
   }
   n14 = Anim Dynamics - Bone: chalk!(ComponentPose=@n13, bAlphaBoolEnabled:bool=@n12, ExternalForce="0,0,0") -> (Pose) {
    alphaInputType=Bool
    angularDampingOverride=0.888
    angularSpringConstant=5000
    bAngularSpring=true
    bOverrideAngularDamping=true
    linearSpringConstant=0.1
    simulationSpace=BoneRelative
    alphaBoolBlend {blendInTime=0.25, blendOutTime=0.2}
    boundBone {boneName=chalk}
    componentAppliedLinearAccClamp=<3000 3000 3000>
    componentLinearAccScale=<0.3 0.3 0.3>
    componentLinearVelScale=<0.3 0.3 0.3>
    gravityOverride {z=-980}
    physicsBodyDefinitions [{
      boundBone {boneName=chalk}
      boxExtents=<20 20 20>
      constraintSetup {angularLimitsMax=<0 0 40>, angularLimitsMin=<0 0 -40>, angularTarget=<1 0 0>}
      localJointOffset=<1 0 0>
     }]
    relativeSpaceBone {boneName=pack}
   }
   n15 = "Transform (Modify) Bone - Bone: chalk"!(ComponentPose=@n14, bAlphaBoolEnabled:bool=@n12, Translation="6,0,0", Rotation="0,0,0", Scale="1,1,1") -> (Pose) {
    alphaInputType=Bool
    translationMode=BMM_Additive
    translationSpace=BCS_BoneSpace
    alphaBoolBlend {blendInTime=0.7, blendOutTime=0.2}
    boneToModify {boneName=chalk}
    translation {x=6}
   }
   n16 = "Transform (Modify) Bone - Bone: brush"!(ComponentPose=@n15, bAlphaBoolEnabled:bool=@n12, Translation="3,0,0", Rotation="0,0,0", Scale="1,1,1") -> (Pose) {
    alphaInputType=Bool
    translationMode=BMM_Additive
    translationSpace=BCS_BoneSpace
    alphaBoolBlend {blendInTime=0.8, blendOutTime=0.2}
    boneToModify {boneName=brush}
    translation {x=3}
   }
   n17 = Anim Dynamics - Bone: stand!(ComponentPose=@n16, Alpha=0.75, GravityOverride:Vector="0,0,-980", ExternalForce="0,0,0") -> (Pose) {
    angularDampingOverride=0.8889
    angularSpringConstant=8000
    bAngularSpring=true
    bGravityOverrideInSimSpace=true
    bLinearSpring=true
    bOverrideAngularDamping=true
    gravityScale=0.35
    linearSpringConstant=0.1
    simulationSpace=RootRelative
    boundBone {boneName=stand}
    componentAppliedLinearAccClamp=<3000 3000 3000>
    componentLinearAccScale=<0.3 0.3 0.3>
    componentLinearVelScale=<0.3 0.3 0.3>
    gravityOverride {z=-980}
    physicsBodyDefinitions [{
      boundBone {boneName=stand}
      boxExtents=<30 30 30>
      constraintSetup {angularTargetAxis=AxisY, angularLimitsMax=<0 10 0>, angularLimitsMin=<0 -60 0>, angularTarget=<0 1 0>}
      localJointOffset=<4 0 2>
     }]
    relativeSpaceBone {boneName=base_board}
    simSpaceSettings {maxAngularAcceleration=30, maxAngularVelocity=3, simSpaceAngularAlpha=1}
   }
   n18 = Get Launcher Jump() -> (Launcher Jump:bool)
   n19 = "Transform (Modify) Bone - Bone: base_board"!(ComponentPose=@n17, bAlphaBoolEnabled:bool=@n18, Translation="0,0,0", Rotation="17,0,0", Scale="1,1,1") -> (Pose) {
    alphaInputType=Bool
    rotationMode=BMM_Additive
    rotationSpace=BCS_BoneSpace
    translationSpace=BCS_BoneSpace
    alphaBoolBlend {blendInTime=0.5, blendOutTime=0.2}
    boneToModify {boneName=base_board}
    rotation {pitch=17}
   }
   n20 = "Transform (Modify) Bone - Bone: stand"!(ComponentPose=@n19, bAlphaBoolEnabled:bool=@n18, Translation="0,0,0", Rotation="13,0,0", Scale="1,1,1") -> (Pose) {
    alphaInputType=Bool
    rotationMode=BMM_Additive
    rotationSpace=BCS_BoneSpace
    translationSpace=BCS_BoneSpace
    alphaBoolBlend {blendInTime=0.5, blendOutTime=0.2}
    boneToModify {boneName=stand}
    rotation {pitch=13}
   }
   Output Pose(Result=Component To Local(ComponentPose=@n20)) {layerGroup=None}
  ]}
 EventGraph {nodes [
   Event BlueprintUpdateAnimation()
   > n1 = Sequence!(execute) -> (then_0, then_1, then_2:Exec, then_3:Exec, then_4:Exec, then_5:Exec, then_6:Exec, then_7:Exec, then_8:Exec)
   n2 = GetOwnerSkeletalMeshComponent()
   n3 = BreakRotator(InRot=BreakTransform(InTransform=GetDeltaTransformFromRefPose(self=@n2, BoneName=spine_05, BaseName=spine_01)).Rotation)
   > Set Backpack Rotate(Backpack Rotate:Rotator=MakeRotator(Pitch=MapRangeClamped(Value=@n3.Yaw, InRangeA=0, InRangeB=-30, OutRangeB=-5))) -> (Output_Get:Rotator)
   > Set Backpack Translation(Backpack Translation:Vector=MakeVector(Y=MapRangeClamped(Value=@n3, InRangeA=-20, InRangeB=20, OutRangeA=0.5, OutRangeB=-3))) -> (Output_Get:Vector)
   n4 = "Gravity Override State Function (Parameter Asset) (Nativized)"(
    execute=@n1.then_1
    CharacterPartAnimInstance=Self-Reference()
    GameTimeInSeconds=GetGameTimeInSeconds()
    DebugJointName=spine_05
    Parameters=&/F_MED_PrizeTray_Pack_Parameters
   )
   > Set Gravity Override(Gravity Override=@n4)
   n5 = Get bIsSprinting()
   n6 = Get bIsTacticalSprinting()
   Set Sprinting(execute=@n1.then_2, Sprinting:double="To Float (Boolean)"(InBool=OR Boolean(A=@n5, B=@n6))) -> (Output_Get:double)
   > Set Melee(Melee:double="To Float (Boolean)"(InBool=Get bIsPlayingMeleeAnim())) -> (Output_Get:double)
   > Set parachute(parachute:double="To Float (Boolean)"(InBool=Get bIsParachuteOpened())) -> (Output_Get:double)
   > Set Dive(Dive:double="To Float (Boolean)"(InBool=Get bIsSkydiveDiveMode())) -> (Output_Get:double)
   > Set float(float:double="To Float (Boolean)"(InBool=Get bIsSkydiveFloating())) -> (Output_Get:double)
   n7 = Get bIsSkyDiving()
   > Set skydive(skydive:double="To Float (Boolean)"(InBool=AND Boolean!(A=@n7, B=false, C:bool, D:bool) -> (ReturnValue))) -> (Output_Get:double)
   Set Launcher Jump(execute=@n1.then_3, Launcher Jump:bool=AND Boolean!(A=Get bIsSkydiveFloating(), B=Get bIsInAir() -> (bIsInAir:bool)) -> (ReturnValue)) -> (Output_Get:bool)
   n8 = Get bIsSkyDiving()
   Set Falling ground(execute=@n1.then_4, Falling ground:bool=AND Boolean(A=NOT Boolean(A=@n8), B=Get bIsFalling())) -> (Output_Get:bool)
   n9 = Branch(execute=@n1.then_7, Condition=Get bIsCrouching())
   > Set neck(neck:double=GetAlphaFromReferenceTransformDelta(
     SkeletalMeshComponent=Get BodyMesh() -> (BodyMesh:Skeletal Mesh Component Object Reference)
     BoneName=head
     BaseName=spine_05
     DeltaComponentType=QuaternionTwist
     InputRange=(X=50,Y=80)
     TwistAxis="0,0,1"
    )) -> (Output_Get:double)
   Set neck(execute=@n9.else, neck=0)
   n10 = Branch(execute=@n1.then_5, Condition=NOT Boolean(A=Get bIsSkyDiving()))
   > Set Cow_Spine(Cow_Spine:double=GetAlphaFromReferenceTransformDelta(
     SkeletalMeshComponent=Get BodyMesh()
     BoneName=spine_05
     BaseName=spine_01
     DeltaComponentType=QuaternionTwist
     InputRange=(X=0,Y=20)
     TwistAxis="0,0,1"
    )) -> (Output_Get:double)
   Set Cow_Spine(execute=@n10.else, Cow_Spine=0)
   Branch(execute=@n1.then_8, Condition=Get bIsPlayingMeleeAnim())
   > n11 = Sequence!(execute) -> (then_0, then_1)
   n12 = BreakTransform(InTransform=GetDeltaTransformFromRefPose(self=GetOwnerSkeletalMeshComponent(), BoneName=clavicle_r, BaseName=spine_05))
   > n13 = Branch(Condition="float <= float"(A=BreakRotator(InRot=@n12.Rotation).Yaw, B=-15) -> (ReturnValue=~))
   > Set UpperArmRot_r(UpperArmRot_r:double=1) -> (Output_Get:double)
   Set UpperArmRot_r(execute=@n13.else, UpperArmRot_r=0)
   n14 = BreakTransform(InTransform=GetDeltaTransformFromRefPose(self=GetOwnerSkeletalMeshComponent(), BoneName=clavicle_r, BaseName=spine_05))
   n15 = Branch(execute=@n11.then_1, Condition="float >= float"(A=BreakRotator(InRot=@n14.Rotation).Yaw, B=15))
   > Set UpperArmRot_l(UpperArmRot_l:double=1) -> (Output_Get:double)
   Set UpperArmRot_l(execute=@n15.else, UpperArmRot_l=0)
   n16 = Branch(execute=@n1.then_6, Condition=NOT Boolean(A=Get bIsSkyDiving()))
   > Set Cat_Spine(Cat_Spine:double=GetAlphaFromReferenceTransformDelta(
     SkeletalMeshComponent=Get BodyMesh()
     BoneName=spine_05
     BaseName=spine_03
     DeltaComponentType=QuaternionTwist
     InputRange=(X=-20,Y=15)
     TwistAxis="0,0,1"
    )) -> (Output_Get:double)
   Set Cat_Spine(execute=@n16.else, Cat_Spine=0)
   Event OnCharacterCustomizationLoaded()
   > n17 = Set Owning Pawn(Owning Pawn:Pawn Object Reference=TryGetPawnOwner()) -> (Output_Get:Pawn Object Reference)
   > Is Valid(InputObject=@n17)
   > n18 = Cast To FortPlayerPawn(Object=Get Owning Pawn() -> (Owning Pawn:Pawn Object Reference))
   > n19 = Set Owning Player Pawn(Owning Player Pawn:Fort Player Pawn Object Reference=@n18) -> (Output_Get:Fort Player Pawn Object Reference)
   > Set BodyMesh(BodyMesh=GetSkeletalMeshForPartType(self=@n19))
   NOT Boolean(A=Get bIsParachuteOpened())
   TryGetPawnOwner()
   NOT Boolean(A=Get bIsSkydiveDiveMode())
  ]}
]
variables [
 Backpack Rotate {type=Rotator}
 Backpack Translation {type=Vector}
 BodyMesh {type=SkeletalMeshComponent}
 Cat_Spine {type=number}
 Cow_Spine {type=number}
 Dive {type=number}
 Falling ground {type=boolean}
 float {type=number}
 Gravity Override {type=Vector}
 Launcher Jump {type=boolean}
 Melee {type=number}
 neck {type=number}
 Owning Pawn {type=Pawn, default=null}
 Owning Player Pawn {type=FortPlayerPawn}
 parachute {type=number}
 Sky Dive Floating Forward {type=boolean}
 skydive {type=number}
 Sprinting {type=number}
 UpperArmRot_l {type=number}
 UpperArmRot_r {type=number}
]
